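# Lobby coachmark widget, a sidebar onboarding tooltip that introduces Gizmos as gear available in the next BR match.
# It wraps a UIKit CoachMark with a "Got it" button and no dismiss button, activating it when the tooltip shows.
# Clicking the button calls the parent's action handler and signals close-finished, as does tooltip end, then deactivates the coachmark.
# IsAvailableForProduct limits it to products matching the Product.BR gameplay tag.
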WidgetBlueprint /FortniteCoreUI/UI/Coachmarks/WBP_Lobby_Coachmark_Consumables
graphs [
 EventGraph {nodes [
   "Button Other Action Clicked (WBP_UIKit_CoachMark)"() {is=title K2Node_ComponentBoundEvent}
   > n1 = HandleOtherActionClicked(self:Fort Sidebar Onboard Tooltip Widget Object Reference)
   NotifyCloseAnimationFinished(execute=[@n1.then, @n2.then], self:Fort Sidebar Onboard Tooltip Widget Object Reference)
   > DeactivateWidget(self=Get WBP_UIKit_CoachMark() -> (WBP_UIKit_CoachMark:WBP UIKit Coach Mark Object Reference))
   Event OnShowTooltip() {is=title K2Node_Event}
   > ActivateWidget(self=Get WBP_UIKit_CoachMark())
   n2 = Event OnEndTooltip() {is=title K2Node_Event}
  ]}
 IsAvailableForProduct {nodes [
   n1 = IsAvailableForProduct!() -> (then, ProductTag:Gameplay Tag Structure) {is=title K2Node_FunctionEntry}
   n2 = Make Array("[0]"=(TagName="Product.BR"))
   > Return Node!(
    execute
    ReturnValue:bool=MatchesAnyTags!(
     TagOne:Gameplay Tag Structure=@n1
     OtherContainer:"Gameplay Tag Container Structure (by ref)"=MakeGameplayTagContainerFromArray!(GameplayTags:Array of Gameplay Tag Structures=@n2) -> (ReturnValue:Gameplay Tag Container Structure)
     bExactMatch:bool=false
    ) -> (ReturnValue:bool)
   )
  ]}
]
variables []
widgets {
 info {
  inheritedWidgetCount=2
  parentClass=/Script/FortniteUI.FortSidebarOnboardTooltipWidget
  rootWidgetClass=/FortUILibrary/UIKit/CoachMark/WBP_UIKit_CoachMark.WBP_UIKit_CoachMark_C
  widgetCount=3
 }
 widgets [
  WBP_UIKit_CoachMark {
   widgetClassPath=/FortUILibrary/UIKit/CoachMark/WBP_UIKit_CoachMark
   bIsVariable=true
   widget=WBP_UIKit_CoachMark
   properties {
    arrowPosition=Left_Down
    shouldShowDismissButton=false
    textDescription="Check out the new gear that'll be available in your next BR match!"
    textHeader=Gizmos
    textOtherActionButton=Got it
    widget Width=657.5338
    input Action Other {dataTable=&InputActionDataTable, rowName=GenericForward}
   }
  }
  Dismiss_Button {widgetClassPath=&CommonButtonLegacy, bInherited=true, widget=null}
  OtherAction_Button {widgetClassPath=&CommonButtonLegacy, bInherited=true, widget=null}
 ]
}
